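# Sound cue for the heavy semi-auto shotgun's fire sound in the Athena mode, built as three distance layers.
# A mixer blends close, medium and distant branches at descending volume, each with its own attenuation, modulator and random pick among wave variations.
# The close layer uses the third-person fire sound class, while the medium and distant layers use the distant fire class.
# Local and global gunfire concurrency sets limit overlapping shots.
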SoundCue /Figment_Shotgun_SemiAuto_Athena/Sounds/FNS01_Weapon_HeavyShotgun_Fire_Athena_Cue
root=Mixer_0
nodes [
 {id=Attenuation_0, children [Modulator_0], properties {attenuationSettings=&Fort_Weapon_Projectile_Close_Athena_Att}}
 {
  id=Attenuation_1
  children [Modulator_1]
  properties {attenuationSettings=&Fort_Weapon_Projectile_Medium_Mono_Athena_Att}
 }
 {
  id=Attenuation_2
  children [Modulator_2]
  properties {attenuationSettings=&Fort_Weapon_Projectile_Distant_Mono_Athena_Att}
 }
 {id=Mixer_0, children [SoundClass_0,SoundClass_6,SoundClass_1], properties {inputVolume [0.8,0.5,0.2]}}
 {id=Modulator_0, children [Random_6], properties {pitchMin=1}}
 {id=Modulator_1, children [Random_8], properties {volumeMax=2, volumeMin=2}}
 {id=Modulator_2, children [Random_9], properties {volumeMax=2, volumeMin=2}}
 {id=Random_6, children [WavePlayer_16, WavePlayer_18, WavePlayer_19], properties {bShouldExcludeFromBranchCulling=true, weights [1,1,1]}}
 {id=Random_8, children [WavePlayer_20, WavePlayer_22, WavePlayer_23], properties {bShouldExcludeFromBranchCulling=true, weights [1,1,1]}}
 {id=Random_9, children [WavePlayer_24, WavePlayer_25], properties {bShouldExcludeFromBranchCulling=true, weights [1,1]}}
 {
  id=SoundClass_0
  children [Attenuation_0]
  properties {soundClassOverride=&Fort_WeaponRanged_Fire_3P_PSM}
 }
 {
  id=SoundClass_1
  children [Attenuation_2]
  properties {soundClassOverride=&Fort_WeaponRanged_Fire_Distant}
 }
 {
  id=SoundClass_6
  children [Attenuation_1]
  properties {soundClassOverride=&Fort_WeaponRanged_Fire_Distant}
 }
 {id=WavePlayer_16, properties {soundWaveAssetPtr=/Game/Sounds/Fort_Weapons/Shotgun/Heavy/Shotgun_NewFire_Close_01}}
 {id=WavePlayer_18, properties {soundWaveAssetPtr=/Game/Sounds/Fort_Weapons/Shotgun/Heavy/Shotgun_NewFire_Close_02}}
 {id=WavePlayer_19, properties {soundWaveAssetPtr=/Game/Sounds/Fort_Weapons/Shotgun/Heavy/Shotgun_NewFire_Close_03}}
 {id=WavePlayer_20, properties {soundWaveAssetPtr=/Game/Sounds/Fort_Weapons/Shotgun/Heavy/Shogun_Fire_MediumNew_01}}
 {id=WavePlayer_22, properties {soundWaveAssetPtr=/Game/Sounds/Fort_Weapons/Shotgun/Heavy/Shogun_Fire_MediumNew_03}}
 {id=WavePlayer_23, properties {soundWaveAssetPtr=/Game/Sounds/Fort_Weapons/Shotgun/Heavy/Shogun_Fire_MediumNew_02}}
 {id=WavePlayer_24, properties {soundWaveAssetPtr=/Game/Sounds/Fort_Weapons/Shotgun/Heavy/Shotgun_Distant_ANew_01}}
 {id=WavePlayer_25, properties {soundWaveAssetPtr=/Game/Sounds/Fort_Weapons/Shotgun/Heavy/Shotgun_Distant_ANew_02}}
]
properties {
 maxDistance=27500
 soundClassObject=&Fort_WeaponRanged_Fire_3P_PSM
 volumeMultiplier=0.9
 attenuationOverrides {falloffDistance=5000, attenuationShapeExtents {x=1000}}
 concurrencyOverrides {bLimitToOwner=true, maxCount=3, resolutionRule=StopOldest}
 concurrencySet [&Weapon_Gunfire_Local_Conc, &Weapon_Gunfire_Global_Conc]
}
