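# Sound cue for inserting a single shell during a semi-auto shotgun reload in the Figment shotgun plugin.
# It plays the shared pump shotgun insert wave after a short fixed delay, then applies slight random pitch variation at constant volume.
# Playback uses the weapon foley sound class and reload attenuation, with a concurrency override that blocks new instances once the limit is reached.
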SoundCue /Figment_Shotgun_SemiAuto_Athena/Sounds/FNS01_Weapon_Shotgun_InsertSingle2_Cue
root=Modulator_13
nodes [
 {id=Delay_0, children [WavePlayer_8], properties {delayMax=0.03, delayMin=0.03}}
 {id=Modulator_13, children [Delay_0], properties {pitchMax=1.1, pitchMin=0.9, volumeMax=1, volumeMin=1}}
 {id=WavePlayer_8, properties {soundWaveAssetPtr=/Game/Sounds/Fort_Weapons/Shotgun/Reload_Elements/weapon_shotgun_pump_insert}}
]
properties {
 attenuationSettings=&Fort_Weapon_Reload_Att
 bOverrideConcurrency=true
 maxDistance=2250
 soundClassObject=&Fort_WeaponFoley
 volumeMultiplier=1
 attenuationOverrides {falloffDistance=1600}
 concurrencyOverrides {resolutionRule=PreventNew}
}
